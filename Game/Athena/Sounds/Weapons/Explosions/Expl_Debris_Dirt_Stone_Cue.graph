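# Sound cue for dirt and stone debris in Athena weapon explosions, routed to the large-explosives sound class with close explosion attenuation.
# A Random node picks one of five equally weighted debris waves, held back by a short fixed Delay and varied by a default Modulator.
# An Enveloper at the root ramps volume up from a quieter start over half a second.
# The shared weapon-explosion concurrency set governs simultaneous plays.
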SoundCue /Game/Athena/Sounds/Weapons/Explosions/Expl_Debris_Dirt_Stone_Cue
root=Enveloper_0
nodes [
 {id=Delay_0, children [Random_3], properties {delayMax=0.1, delayMin=0.1}}
 {
  id=Enveloper_0
  children [Modulator_2]
  properties {pitchCurve {editorCurveData {keys [{time=0}]}}, volumeCurve {editorCurveData {keys [<at 0 0.25>,<at 0.5 1>]}}}
 }
 {id=Modulator_2, children [Delay_0]}
 {id=Random_3, children [WavePlayer_0, WavePlayer_1, WavePlayer_4, WavePlayer_10, WavePlayer_11], properties {weights [1,1,1,1,1]}}
 {id=WavePlayer_0, properties {soundWaveAssetPtr=&/EXPL_Debris_Dirt_Stone_05}}
 {id=WavePlayer_1, properties {soundWaveAssetPtr=&/EXPL_Debris_Dirt_Stone_01}}
 {id=WavePlayer_10, properties {soundWaveAssetPtr=&/EXPL_Debris_Dirt_Stone_03}}
 {id=WavePlayer_11, properties {soundWaveAssetPtr=&/EXPL_Debris_Dirt_Stone_04}}
 {id=WavePlayer_4, properties {soundWaveAssetPtr=&/EXPL_Debris_Dirt_Stone_02}}
]
properties {
 attenuationSettings=/Game/Sounds/Attenuation/Fort_Explosion_Close_Athena_Att
 maxDistance=1499.9875
 soundClassObject=/Game/Sounds/FortSoundClassesAndModes/Fort_Explosives_Large_PSM
 volumeMultiplier=0.7076
 concurrencyOverrides {maxCount=4, resolutionRule=StopOldest, volumeScale=0.5}
 concurrencySet [/Game/Sounds/Concurrency_Rules/Weapon_Explosion_Conc]
}
