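# Sound cue for metal debris in Athena weapon explosions, routed to the large explosives sound class with close-explosion attenuation.
# It picks one of five equally weighted metal debris waves and plays it after a short fixed delay, through a default Modulator for variation.
# A root Enveloper ramps volume up from a quarter to full over the first half second.
# A shared weapon explosion concurrency set is assigned to manage overlapping instances.
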SoundCue /Game/Athena/Sounds/Weapons/Explosions/Expl_Debris_Metal_Cue
root=Enveloper_1
nodes [
 {id=Delay_1, children [Random_1], properties {delayMax=0.1, delayMin=0.1}}
 {
  id=Enveloper_1
  children [Modulator_1]
  properties {pitchCurve {editorCurveData {keys [{time=0}]}}, volumeCurve {editorCurveData {keys [<at 0 0.25>,<at 0.5 1>]}}}
 }
 {id=Modulator_1, children [Delay_1]}
 {id=Random_1, children [WavePlayer_0,WavePlayer_1,WavePlayer_4,WavePlayer_8,WavePlayer_9], properties {weights [1,1,1,1,1]}}
 {id=WavePlayer_0, properties {soundWaveAssetPtr=&/EXPL_Debris_Metal_05}}
 {id=WavePlayer_1, properties {soundWaveAssetPtr=&/EXPL_Debris_Metal_01}}
 {id=WavePlayer_4, properties {soundWaveAssetPtr=&/EXPL_Debris_Metal_02}}
 {id=WavePlayer_8, properties {soundWaveAssetPtr=&/EXPL_Debris_Metal_03}}
 {id=WavePlayer_9, properties {soundWaveAssetPtr=&/EXPL_Debris_Metal_04}}
]
properties {
 attenuationSettings=/Game/Sounds/Attenuation/Fort_Explosion_Close_Athena_Att
 maxDistance=1499.9875
 soundClassObject=/Game/Sounds/FortSoundClassesAndModes/Fort_Explosives_Large_PSM
 volumeMultiplier=0.7076
 concurrencyOverrides {maxCount=4, resolutionRule=StopOldest, volumeScale=0.5}
 concurrencySet [/Game/Sounds/Concurrency_Rules/Weapon_Explosion_Conc]
}
