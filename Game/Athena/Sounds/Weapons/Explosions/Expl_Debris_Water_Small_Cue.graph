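# Sound cue for small explosion debris hitting water in Athena, played as a medium explosives sound class with close water-explosion attenuation.
# It mixes a random grenade water-impact wave at full level with a random, much quieter debris water-wave layer.
# The mix passes through a short fixed delay, then a default-setting modulator at the root.
# A concurrency override limits simultaneous instances and scales their volume.
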SoundCue /Game/Athena/Sounds/Weapons/Explosions/Expl_Debris_Water_Small_Cue
root=Modulator_2
nodes [
 {id=Delay_0, children [Mixer_0], properties {delayMax=0.08, delayMin=0.08}}
 {id=Mixer_0, children [Random_2,Random_0], properties {inputVolume [1,0.1]}}
 {id=Modulator_2, children [Delay_0]}
 {id=Random_0, children [WavePlayer_2,WavePlayer_3,WavePlayer_5,WavePlayer_6], properties {weights [1,1,1,1]}}
 {id=Random_2, children [WavePlayer_7,WavePlayer_8,WavePlayer_9], properties {weights [1,1,1]}}
 {id=WavePlayer_2, properties {soundWaveAssetPtr=&/EXPL_Debris_Water_Wave_04}}
 {id=WavePlayer_3, properties {soundWaveAssetPtr=&/EXPL_Debris_Water_Wave_01}}
 {id=WavePlayer_5, properties {soundWaveAssetPtr=&/EXPL_Debris_Water_Wave_02}}
 {id=WavePlayer_6, properties {soundWaveAssetPtr=&/EXPL_Debris_Water_Wave_03}}
 {id=WavePlayer_7, properties {soundWaveAssetPtr=/Game/Sounds/Fort_GamePlay_Sounds/Explosives/Grenade_Impact_Water_01}}
 {id=WavePlayer_8, properties {soundWaveAssetPtr=/Game/Sounds/Fort_GamePlay_Sounds/Explosives/Grenade_Impact_Water_02}}
 {id=WavePlayer_9, properties {soundWaveAssetPtr=/Game/Sounds/Fort_GamePlay_Sounds/Explosives/Grenade_Impact_Water_03}}
]
properties {
 attenuationSettings=/Game/Sounds/Attenuation/Fort_Explosion_Water_Close_Athena_Att
 bOverrideConcurrency=true
 maxDistance=7241.2505
 soundClassObject=/Game/Sounds/FortSoundClassesAndModes/Fort_Explosives_Medium_PSM
 volumeMultiplier=0.5
 concurrencyOverrides {maxCount=3, volumeScale=0.5}
}
