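# Sound cue for wood debris in weapon explosions, using close-explosion attenuation, the large explosives sound class and a weapon explosion concurrency set.
# It picks one of five equally weighted debris waves, starts it after a short fixed delay and passes it through a default Modulator for pitch and volume variation.
# An Enveloper at the root ramps volume up from a quarter to full over half a second.
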SoundCue /Game/Athena/Sounds/Weapons/Explosions/Expl_Debris_Wood_Cue
root=Enveloper_0
nodes [
 {id=Delay_1, children [Random_0], properties {delayMax=0.1, delayMin=0.1}}
 {
  id=Enveloper_0
  children [Modulator_0]
  properties {pitchCurve {editorCurveData {keys [{time=0}]}}, volumeCurve {editorCurveData {keys [<at 0 0.25>,<at 0.5 1>]}}}
 }
 {id=Modulator_0, children [Delay_1]}
 {id=Random_0, children [WavePlayer_2,WavePlayer_3,WavePlayer_5,WavePlayer_6,WavePlayer_7], properties {weights [1,1,1,1,1]}}
 {id=WavePlayer_2, properties {soundWaveAssetPtr=&/EXPL_Debris_Wood_01}}
 {id=WavePlayer_3, properties {soundWaveAssetPtr=&/EXPL_Debris_Wood_02}}
 {id=WavePlayer_5, properties {soundWaveAssetPtr=&/EXPL_Debris_Wood_03}}
 {id=WavePlayer_6, properties {soundWaveAssetPtr=&/EXPL_Debris_Wood_04}}
 {id=WavePlayer_7, properties {soundWaveAssetPtr=&/EXPL_Debris_Wood_05}}
]
properties {
 attenuationSettings=/Game/Sounds/Attenuation/Fort_Explosion_Close_Athena_Att
 maxDistance=1499.9875
 soundClassObject=/Game/Sounds/FortSoundClassesAndModes/Fort_Explosives_Large_PSM
 volumeMultiplier=0.7076
 concurrencyOverrides {maxCount=4, resolutionRule=StopOldest, volumeScale=0.5}
 concurrencySet [/Game/Sounds/Concurrency_Rules/Weapon_Explosion_Conc]
}
