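# Sound cue for the frag grenade explosion in Athena, blending close, medium and distant layers by listener distance.
# A mixer combines the three layers, each randomly picking one of four waves through volume or pitch modulators and its own distance attenuation asset with overridden falloff.
# The close layer uses the third-person fire sound class, while medium and distant share the distant fire class.
# A concurrency override caps simultaneous instances and scales their volume.
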SoundCue /Game/Athena/Sounds/Weapons/Explosions/Grenade_Expl_Cue
root=Mixer_4
nodes [
 {
  id=Attenuation_0
  children [Modulator_1]
  properties {
   attenuationSettings=/Game/Sounds/Attenuation/Fort_Explosion_Distant_Athena_Att
   attenuationOverrides {falloffDistance=3000, attenuationShapeExtents {x=1000}}
  }
 }
 {
  id=Attenuation_4
  children [Modulator_2]
  properties {attenuationSettings=/Game/Sounds/Attenuation/Fort_Explosion_Close_Athena_Att, attenuationOverrides {falloffDistance=1600}}
 }
 {
  id=Attenuation_5
  children [Modulator_5]
  properties {
   attenuationSettings=/Game/Sounds/Attenuation/Fort_Explosion_Medium_Athena_Att
   attenuationOverrides {falloffDistance=3000, attenuationShapeExtents {x=1000}}
  }
 }
 {id=Mixer_4, children [SoundClass_3,SoundClass_2,SoundClass_0], properties {inputVolume [1,0.8,0.65]}}
 {id=Modulator_0, children [Attenuation_4]}
 {id=Modulator_1, children [Random_5], properties {pitchMax=0.95, pitchMin=0.85, volumeMax=1.1, volumeMin=1}}
 {id=Modulator_2, children [Random_3], properties {pitchMax=1, pitchMin=1, volumeMax=1.6, volumeMin=1.6}}
 {id=Modulator_3, children [Random_4], properties {volumeMax=1.2, volumeMin=1.2}}
 {id=Modulator_5, children [Modulator_3], properties {pitchMax=1.1, pitchMin=0.9, volumeMax=1, volumeMin=0.9}}
 {id=Random_3, children [WavePlayer_0, WavePlayer_1, WavePlayer_4, WavePlayer_10], properties {weights [1,1,1,1]}}
 {id=Random_4, children [WavePlayer_11, WavePlayer_12, WavePlayer_13, WavePlayer_14], properties {weights [1,1,1,1]}}
 {id=Random_5, children [WavePlayer_15, WavePlayer_16, WavePlayer_17, WavePlayer_18], properties {weights [1,1,1,1]}}
 {
  id=SoundClass_0
  children [Attenuation_0]
  properties {soundClassOverride=&Fort_WeaponRanged_Fire_Distant}
 }
 {
  id=SoundClass_2
  children [Attenuation_5]
  properties {soundClassOverride=&Fort_WeaponRanged_Fire_Distant}
 }
 {id=SoundClass_3, children [Modulator_0], properties {soundClassOverride=&Fort_WeaponRanged_Fire_3P_PSM}}
 {id=WavePlayer_0, properties {soundWaveAssetPtr=&/Frag_Grenade_Explosion_Close_01}}
 {id=WavePlayer_1, properties {soundWaveAssetPtr=&/Frag_Grenade_Explosion_Close_02}}
 {id=WavePlayer_10, properties {soundWaveAssetPtr=&/Frag_Grenade_Explosion_Close_04}}
 {id=WavePlayer_11, properties {soundWaveAssetPtr=&/Frag_Grenade_Explosion_Medium_01}}
 {id=WavePlayer_12, properties {soundWaveAssetPtr=&/Frag_Grenade_Explosion_Medium_02}}
 {id=WavePlayer_13, properties {soundWaveAssetPtr=&/Frag_Grenade_Explosion_Medium_03}}
 {id=WavePlayer_14, properties {soundWaveAssetPtr=&/Frag_Grenade_Explosion_Medium_04}}
 {id=WavePlayer_15, properties {soundWaveAssetPtr=&/Frag_Grenade_Explosion_Distant_01}}
 {id=WavePlayer_16, properties {soundWaveAssetPtr=&/Frag_Grenade_Explosion_Distant_02}}
 {id=WavePlayer_17, properties {soundWaveAssetPtr=&/Frag_Grenade_Explosion_Distant_03}}
 {id=WavePlayer_18, properties {soundWaveAssetPtr=&/Frag_Grenade_Explosion_Distant_04}}
 {id=WavePlayer_4, properties {soundWaveAssetPtr=&/Frag_Grenade_Explosion_Close_03}}
]
properties {bOverrideConcurrency=true, maxDistance=27500, volumeMultiplier=1.35, concurrencyOverrides {maxCount=3, volumeScale=0.5}}
